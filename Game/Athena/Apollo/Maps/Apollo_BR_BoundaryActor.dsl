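; Boundary actor for the Apollo BR map that tracks the current playlist and can switch off the map's boundary volumes.
; On BeginPlay it binds to the Fortnite game state's playlist-loaded delegate, whose handler only stores the playlist name.
; TurnOffBlockingKillAndBoundaries disables collision on listed KillZ and blocking volumes and destroys the fly boundary spline, but nothing here calls it.
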(event EventOnWorldReady)

(event EventBeginPlay
  ;[1 1 1] SERVER/CLIENT - Bind when playlist changes
  (CallFunction|BindPlaylistHasChanged))

(event Custom|BindPlaylistHasChanged
  ;[1 1 1] SERVER/CLIENT - Bind to when playlist has changed to try and clear blocking and boundary
  (bind _asfort_game_state_athena (Utilities|GetValidFortAthenaGameState)
    (:Valid
      (Playlist|AssignOnCurrentPlaylistLoaded _asfort_game_state_athena (delegate Custom|PlaylistHasLoaded)))))

(event Custom|PlaylistHasLoaded
  (params (PlaylistName :type Name) (PlaylistContextTags :type Gameplay%20Tag%20Container%20Structure%20%28by%20ref%29))
  ;[1 1 1] SERVER/CLIENT - Bind to when playlist has changed to try and clear blocking and boundary
  (Variables|BoundryActorMembers|SetCurrentPlaylistName PlaylistName))

(fn GetFortAthenaGameState
  (bind _asfort_game_state_athena (Utilities|Casting|CastToFortGameStateAthena (Game|GetGameState))
    (:then
      (return _asfort_game_state_athena))
    (:CastFailed
      (return 0))))

(fn TurnOffBlockingKillAndBoundaries
  (for _array_element (Variables|BoundaryActors|GetKillZList)
    (Utilities|IsValid _array_element
      (:Is%20Valid
        (Collision|SetCollisionEnabled (Variables|/Script/Engine.Brush|GetBrushComponent _array_element) "NoCollision")
        (Collision|SetCollisionProfileName (Variables|/Script/Engine.Brush|GetBrushComponent _array_element) (LocalVariables|GetCONST_CollisionProfileNoCollision) true :!class /Script/Engine.PrimitiveComponent))))
  (for _array_element_2 (Variables|BoundaryActors|GetBlockingVolumeList)
    (Utilities|IsValid _array_element_2
      (:Is%20Valid
        (Collision|SetCollisionEnabled (Variables|/Script/Engine.Brush|GetBrushComponent _array_element_2) "NoCollision")
        (Collision|SetCollisionProfileName (Variables|/Script/Engine.Brush|GetBrushComponent _array_element_2) (LocalVariables|GetCONST_CollisionProfileNoCollision) true :!class /Script/Engine.PrimitiveComponent))))
  (Actor|DestroyActor (Variables|BoundaryActors|GetFlyBoundrySpline)))

(fn ConstructionScript)
